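# Leaderboard tab button widget built on CommonButtonLegacy, showing a left icon and an optional centered label.
# SetTabLabelInfo feeds display name and icon brush into Set Text and Set Icon, while PreConstruct collapses the icon when its brush draws nothing.
# Set Text shows the label and adds an icon tooltip only if text is non-empty and UseText is on.
# Hover and selection events tint the icon, and text style follows button state.
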WidgetBlueprint /Game/Athena/UI/Leaderboards/AthenaLeaderboardTabButton
graphs [
 EventGraph {nodes [
   Event On Hovered()
   > SetColorAndOpacity(
    self:Image Object Reference=Get LeftSideImage() -> (LeftSideImage:Image Object Reference)
    InColorAndOpacity:Linear Color Structure=Get HoveredIconTint() -> (HoveredIconTint:Linear Color Structure)
   )
   n1 = Event SetTabLabelInfo!() -> (OutputDelegate, then, TabLabelInfo:"Fort Tab Button Label Info Structure (by ref)") {is=title K2Node_Event}
   > n2 = Sequence()
   n3 = Break Fort Tab Button Label Info!(FortTabButtonLabelInfo:"Fort Tab Button Label Info Structure (by ref)"=@n1.TabLabelInfo) -> (DisplayName:Text, IconBrush:Slate Brush Structure, IconResource:Soft Object Reference) {is=title K2Node_BreakStruct}
   > Set Text!(execute, self, ButtonText:Text=@n3) -> (then) {is=title K2Node_CallFunction}
   Set Icon!(execute=@n2.then_1, self, IconBrush:Slate Brush Structure=@n3.IconBrush) -> (then) {is=title K2Node_CallFunction}
   Event OnCurrentTextStyleChanged() {is=title K2Node_Event}
   > SetStyle!(
    execute
    self:Common Text Object Reference=Get CenterButtonTextWidget() -> (CenterButtonTextWidget:Common Text Object Reference)
    InStyle:Common Text Style Class Reference=GetCurrentTextStyleClass!(self:Common Button Base Object Reference) -> (ReturnValue:Common Text Style Class Reference)
   ) -> (then)
   Event On Unhovered()
   > n4 = Branch(Condition=GetSelected!(self:Common Button Base Object Reference) -> (ReturnValue:bool))
   n5 = Get LeftSideImage()
   > SetColorAndOpacity(self=@n5, InColorAndOpacity=Get HoveredIconTint())
   SetColorAndOpacity(execute=@n4.else, self=@n5, InColorAndOpacity=Get DeselectedIconTint() -> (DeselectedIconTint:Linear Color Structure))
   Event Construct()
   > n6 = Branch(Condition=GetSelected())
   n7 = Get LeftSideImage()
   > SetColorAndOpacity(self=@n7, InColorAndOpacity=Get HoveredIconTint())
   SetColorAndOpacity(execute=@n6.else, self=@n7, InColorAndOpacity=Get DeselectedIconTint())
   Event PreConstruct()
   > n8 = Sequence()
   Set Text(execute=@n8.then_1, ButtonText=Get ButtonText() -> (ButtonText:Text))
   > SetStyle(self=Get CenterButtonTextWidget(), InStyle=GetCurrentTextStyleClass())
   n9 = Get IconBrush() -> (IconBrush:Slate Brush Structure)
   n10 = Branch(
    execute=@n8.then_0
    Condition="Equal (Enum)"(
     A:ESlateBrushDrawType Enum=Break Slate Brush!(SlateBrush:"Slate Brush Structure (by ref)"=@n9) -> (
      TintColor:Slate Color Structure
      DrawAs:ESlateBrushDrawType Enum
      Tiling:ESlateBrushTileType Enum
      Mirroring:ESlateBrushMirrorType Enum
      ImageSize:Vector 2D Structure
      Margin:Margin Structure
      ResourceObject:Object Reference
      OutlineSettings:Slate Brush Outline Settings Structure
     ) {is=title K2Node_BreakStruct}.DrawAs
     B:ESlateBrushDrawType Enum=NoDrawType
    )
   )
   > SetVisibility(self=Get LeftSideImage(), InVisibility=Collapsed)
   SetBrush(execute=@n10.else, self=Get LeftSideImage(), InBrush=Get IconBrush())
   Event On Selected()
   > SetColorAndOpacity(self=Get LeftSideImage(), InColorAndOpacity=Get SelectedIconTint() -> (SelectedIconTint:Linear Color Structure))
   Event On Deselected()
   > SetColorAndOpacity(self=Get LeftSideImage(), InColorAndOpacity=Get DeselectedIconTint())
  ]}
 Set Icon {nodes [
   n1 = Set Icon!() -> (then, IconBrush:Slate Brush Structure) {is=title K2Node_FunctionEntry}
   > n2 = Set IconBrush(IconBrush:Slate Brush Structure=@n1) -> (Output_Get:Slate Brush Structure)
   > SetBrush(self=Get LeftSideImage(), InBrush=@n2)
  ]}
 Set Text {nodes [
   n1 = Set Text!() -> (then, ButtonText:Text) {is=title K2Node_FunctionEntry}
   > n2 = Set ButtonText(ButtonText:Text=@n1) -> (Output_Get:Text)
   n3 = TextIsEmpty(InText=@n1)
   > n4 = Branch(Condition=AND Boolean(A=NOT Boolean(A=@n3), B=Get UseText() -> (UseText:bool)))
   n5 = Get CenterButtonTextWidget()
   > SetVisibility(self=@n5, InVisibility=SelfHitTestInvisible)
   > "SetText (Text)"(self=@n5, InText=@n2)
   > n6 = Create Basic Tooltip!(
    execute
    Owning Player:Player Controller Object Reference=GetOwningPlayer()
    Body Text:Text
    Header Text:Text=Get ButtonText()
    Icon Brush:Slate Brush Asset Object Reference
    Header Style:Common Text Style Class Reference
    Body Style:Common Text Style Class Reference
   ) -> (then, Output:User Widget Object Reference)
   > SetToolTip!(execute, self:Widget Object Reference=Get LeftSideImage(), Widget:Widget Object Reference=@n6) -> (then)
   SetVisibility(execute=@n4.else, self=@n5, InVisibility=Collapsed)
  ]}
 ShowText {nodes [
   ShowText() {is=title K2Node_FunctionEntry}
   > Set UseText(UseText:bool=true) -> (Output_Get:bool)
   > Set Text(ButtonText=Get ButtonText()) {is=title K2Node_CallFunction}
  ]}
]
variables [
 ButtonText {type=string, default="{Text}"}
 DeselectedIconTint {type=LinearColor, default=<rgba 1 1 1 0.4>}
 HoveredIconTint {type=LinearColor, default=<rgba 1 1 1 1>}
 IconBrush {
  type=SlateBrush
  default {
   bIsDynamicallyLoaded=false
   drawAs=NoDrawType
   imageType=NoImage
   mirroring=NoMirror
   resourceName=None
   resourceObject=null
   tiling=NoTile
   imageSize=<0 0>
   margin=<ltrb 0 0 0 0>
   outlineSettings {
    bUseBrushTransparency=false
    roundingType=HalfHeightRadius
    width=0
    color {colorUseRule=UseColor_Specified, specifiedColor=<rgba 0 0 0 0>}
    cornerRadii=<0 0 0 1>
   }
   tintColor {colorUseRule=UseColor_Specified, specifiedColor=<rgba 1 1 1 1>}
   uVRegion {bIsValid=false, max=<0 0>, min=<0 0>}
  }
 }
 SelectedIconTint {type=LinearColor, default=<rgba 1 1 1 1>}
 UseText {type=boolean, default=true}
]
widgets {
 info {inheritedWidgetCount=1, parentClass=&CommonButtonLegacy, rootWidgetClass=&Overlay, widgetCount=6}
 widgets [
  Overlay_1 {widgetClassPath=&Overlay, widget=Overlay_1}
  SizeBox_0 {
   widgetClassPath=&SizeBox
   parent=Overlay_1
   widget=SizeBox_0
   properties {bOverride_MaxDesiredWidth=true, maxDesiredWidth=250}
   slot {class=OverlaySlot, properties {horizontalAlignment=HAlign_Fill, verticalAlignment=VAlign_Fill}}
  }
  ContentHB {widgetClassPath=&HorizontalBox, bIsVariable=true, parent=SizeBox_0, widget=ContentHB, slot {class=SizeBoxSlot}}
  LeftSideImage {
   widgetClassPath=&Image
   bIsVariable=true
   parent=ContentHB
   widget=LeftSideImage
   properties {brush {
     bIsDynamicallyLoaded=false
     drawAs=NoDrawType
     imageType=NoImage
     mirroring=NoMirror
     resourceName=None
     resourceObject=null
     tiling=NoTile
     imageSize=<0 0>
     margin=<ltrb 0 0 0 0>
     outlineSettings {
      bUseBrushTransparency=false
      roundingType=HalfHeightRadius
      width=0
      color {colorUseRule=UseColor_Specified, specifiedColor=<rgba 0 0 0 0>}
      cornerRadii=<0 0 0 0>
     }
     tintColor {colorUseRule=UseColor_Specified, specifiedColor=<rgba 1 1 1 1>}
     uVRegion {bIsValid=false, max=<0 0>, min=<0 0>}
    }}
   slot {class=HorizontalBoxSlot, properties {horizontalAlignment=HAlign_Left, verticalAlignment=VAlign_Center, padding=<ltrb 4 4 4 4>}}
  }
  CenterButtonTextWidget {
   widgetClassPath=&CommonTextBlock
   bIsVariable=true
   parent=ContentHB
   widget=CenterButtonTextWidget
   properties {
    justification=Center
    scrollStyle=&TextScrollStyle-Base_C
    style=/FortUILibrary/Styles/Typography/Foundation/TextStyle-Base-S-B.TextStyle-Base-S-B_C
    text="{Text}"
    visibility=Collapsed
    font {
     bForceMonospaced=false
     bMaterialIsStencil=false
     fontMaterial=null
     fontObject=&NotoSans
     letterSpacing=0
     monospacedWidth=1
     size=18
     skewAmount=0
     typefaceFontName=NotoSans-Bold
     outlineSettings {
      bApplyOutlineToDropShadows=false
      bMiteredCorners=false
      bSeparateFillAlpha=false
      outlineMaterial=null
      outlineSize=0
      outlineColor=<rgba 0 0 0 1>
     }
    }
    shadowOffset=<0 0>
   }
   slot {
    class=HorizontalBoxSlot
    properties {horizontalAlignment=HAlign_Center, verticalAlignment=VAlign_Center, padding=<ltrb 4 4 4 4>, size {sizeRule=Fill, value=1}}
   }
  }
  InputActionWidget {widgetClassPath=&CommonActionWidget, bInherited=true, widget=null}
 ]
}
